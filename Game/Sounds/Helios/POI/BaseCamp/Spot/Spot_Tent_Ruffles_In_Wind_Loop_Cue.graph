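# Looping spot ambience cue of tent fabric ruffling in wind, for the Helios BaseCamp POI.
# A random node picks one of two looping waves, feeding two chained, endlessly looping envelopers that modulate pitch and volume over 30- and 20-second cycles.
# Overridden attenuation adds a custom falloff curve, HRTF, low-pass, occlusion and listener focus.
# Concurrency allows one instance, and output also sends to the exterior ambient source bus.
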SoundCue /Game/Sounds/Helios/POI/BaseCamp/Spot/Spot_Tent_Ruffles_In_Wind_Loop_Cue
root=Enveloper_2
nodes [
 {
  id=Enveloper_1
  children [Random_0]
  properties {
   bLoop=true
   bLoopIndefinitely=true
   loopEnd=30
   pitchCurve {editorCurveData {keys [
      {arriveTangent=-0.0013, interpMode=RCIM_Cubic, leaveTangent=-0.0013, time=0}
      <at 15 0.98 RCIM_Cubic>
      {arriveTangent=0.0013, interpMode=RCIM_Cubic, leaveTangent=0.0013, time=30}
     ]}}
   volumeCurve {editorCurveData {keys [<at 0 1>]}}
  }
 }
 {
  id=Enveloper_2
  children [Enveloper_1]
  properties {
   bLoop=true
   bLoopIndefinitely=true
   loopEnd=20
   pitchCurve {editorCurveData {keys [{time=0}]}}
   volumeCurve {editorCurveData {keys [
      {arriveTangent=-0.01, interpMode=RCIM_Cubic, time=0, value=1}
      <at 10 1.1 RCIM_Cubic>
      {interpMode=RCIM_Cubic, leaveTangent=0.01, time=20, value=1}
     ]}}
  }
 }
 {id=Random_0, children [WavePlayer_1,WavePlayer_2], properties {weights [1,1]}}
 {id=WavePlayer_1, properties {bLooping=true, soundWaveAssetPtr=&/Spot_Tent_Ruffles_In_Wind_Loop_01}}
 {id=WavePlayer_2, properties {bLooping=true, soundWaveAssetPtr=&/Spot_Tent_Ruffles_In_Wind_Loop_02}}
]
properties {
 bOverrideAttenuation=true
 bOverrideConcurrency=true
 maxDistance=1500
 soundClassObject=/Game/Sounds/FortSoundClassesAndModes/Fort_Water
 volumeMultiplier=0.25
 attenuationOverrides {
  bAttenuateWithLPF=true
  bEnableFocusInterpolation=true
  bEnableListenerFocus=true
  bEnableReverbSend=false
  distanceAlgorithm=Custom
  falloffDistance=1500
  focusAttackInterpSpeed=0.01
  focusReleaseInterpSpeed=0.01
  hPFFrequencyAtMax=250
  hPFFrequencyAtMin=250
  lPFFrequencyAtMax=5000
  lPFRadiusMax=2250
  lPFRadiusMin=1000
  nonFocusVolumeAttenuation=0.5
  nonSpatializedRadiusEnd=1000
  nonSpatializedRadiusStart=1500
  occlusionInterpolationTime=2
  occlusionLowPassFilterFrequency=750
  occlusionTraceChannel=ECC_GameTraceChannel11
  occlusionVolumeAttenuation=0.25
  spatializationAlgorithm=SPATIALIZATION_HRTF
  stereoSpread=1024
  attenuationShapeExtents=<0 600>
  customAttenuationCurve {editorCurveData {keys [<at 0.25 0.75 RCIM_Cubic>, <at 0.5 1>, {interpMode=RCIM_Cubic, leaveTangent=-2, time=0.8, value=1}, <at 1 0>]}}
 }
 busSends [{audioBus=null, minSendLevel=0, sendLevel=1, soundSourceBus=&AmbientAudio_Exterior_SourceBus}]
 concurrencyOverrides {maxCount=1, resolutionRule=PreventNew}
}
